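; Movement mode extension for roll landing that grants the owning pawn energy (stamina) when it activates.
; On activation, if the pawn has an energy component and the scalable-float gain toggle is on, it asks the controller's roll landing component to update and return the regen amount.
; That amount is added to the energy component with a tactical sprint HUD gain-highlight tag.
; Deactivation only calls the parent.
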(event EventOnExtensionActivate
  (params (RuntimeContext :type MMERuntime%20Context%20Structure%20%28by%20ref%29))
  (bind _ownerpawn (Variables|FortMovementMode_BaseExtLogic|GetOwnerPawn))
  (bind _returnvalue (Actor|GetComponentbyClass _ownerpawn "/Script/EnergyRuntime.FortComponent_Energy"))
  (|CallParentFunction RuntimeContext)
  (Utilities|IsValid _returnvalue
    (:Is%20Valid
      (if (ScalableFloat|GetValueasBool (Variables|Default|GetEnergyGainEnabled))
        (bind _staminaamount (Class|BPFortControllerComponentRollLanding|UpdateandGetStaminaRegen (Actor|GetComponentbyClass (Pawn|GetController (Variables|FortMovementMode_BaseExtLogic|GetOwnerPawn)) "/RollLanding/Components/BP_FortControllerComponent_RollLanding.BP_FortControllerComponent_RollLanding_C")))
        (Energy|AddEnergy _returnvalue _staminaamount 0 "(GameplayTags=((TagName=\"HUD.Athena.TacticalSprint.HighlightGain\")))")))))

(event EventOnExtensionDeactivate
  (params (RuntimeContext :type MMERuntime%20Context%20Structure%20%28by%20ref%29))
  (|CallParentFunction RuntimeContext))
